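# UI-domain material for the video player widget in the frontend Simple Build and Edit modals, judging by its name and folder.
# It is unlit and translucent, with a single texture sample of WidgetMediaPlayer_Video_SB, which appears to be the media player's video texture.
# The graph contains no other nodes, such as parameters, tinting or UV math.
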Material /FortniteCoreUI/Frontend/SimpleBuildAndEditModals/VideoFiles/VideoPlayerWidgetMaterial_Unique_SB
graph {
 properties {
  bUsedWithStaticMesh=false
  blendMode=BLEND_Translucent
  materialDomain=MD_UI
  pixelDepthOffsetMode=PDOM_Legacy
  shadingModel=MSM_Unlit
  usedShadingModels=Unlit
 }
 nodes [TextureSample() {texture=&/WidgetMediaPlayer_Video_SB}]
}
