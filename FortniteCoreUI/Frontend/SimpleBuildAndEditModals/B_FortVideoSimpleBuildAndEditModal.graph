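# Video widget for the Simple Build and Edit frontend modal, a child of B_FortVideoWidgetBase that adds no widgets of its own.
# On activation it resizes the video texture image to a fixed portrait size.
# When ShouldOverrideImageVideoTextureBrush is set, it then swaps the image brush for ImageVideoTextureBrushOverride, which defaults to a video player material unique to this modal.
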WidgetBlueprint /FortniteCoreUI/Frontend/SimpleBuildAndEditModals/B_FortVideoSimpleBuildAndEditModal
graphs [EventGraph {nodes [
   Event On Activated()
   > OverrideImageVideoTextureSize!(execute, self:B Fort Video Widget Base Object Reference, SizeX:double=512, SizeY:double=700) -> (then)
   > Branch(Condition=Get ShouldOverrideImageVideoTextureBrush() -> (ShouldOverrideImageVideoTextureBrush:bool))
   > OverrideImageVideoTextureBrush!(
    execute
    self:B Fort Video Widget Base Object Reference
    Brush:Slate Brush Structure=Get ImageVideoTextureBrushOverride() -> (ImageVideoTextureBrushOverride:Slate Brush Structure)
   ) -> (then)
  ]}]
variables [
 ImageVideoTextureBrushOverride {
  type=SlateBrush
  default {
   bIsDynamicallyLoaded=false
   drawAs=Image
   imageType=FullColor
   mirroring=NoMirror
   resourceName=None
   resourceObject=/FortniteCoreUI/Frontend/SimpleBuildAndEditModals/VideoFiles/VideoPlayerWidgetMaterial_Unique_SB
   tiling=NoTile
   imageSize=<1920 1080>
   margin=<ltrb 0 0 0 0>
   outlineSettings {
    bUseBrushTransparency=false
    roundingType=HalfHeightRadius
    width=0
    color {colorUseRule=UseColor_Specified, specifiedColor=<rgba 0 0 0 0>}
    cornerRadii=<0 0 0 0>
   }
   tintColor {colorUseRule=UseColor_Specified, specifiedColor=<rgba 1 1 1 1>}
   uVRegion {bIsValid=false, max=<0 0>, min=<0 0>}
  }
 }
 ShouldOverrideImageVideoTextureBrush {type=boolean}
]
widgets {
 info {
  inheritedWidgetCount=0
  parentClass=/Game/UI/Foundation/Movie/B_FortVideoWidgetBase.B_FortVideoWidgetBase_C
  rootWidgetClass=null
  widgetCount=0
 }
 widgets []
}
