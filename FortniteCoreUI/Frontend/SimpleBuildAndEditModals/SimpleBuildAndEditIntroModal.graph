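# Activatable dialog modal introducing Simple Build and Simple Edit, with explanatory text, videos and an inline MVVM settings list.
# SimpleBuildAndEditListViewModel reports which settings to show, toggling each section and video, then receives the active video widget and selection changes.
# A dialog viewmodel fills the UIKit dialog, and Go Settings deactivates the modal and opens Gameplay settings at Simple Build.
# Compact UI tightens spacing; a key-binding hint appears only on listed non-mobile platforms.
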WidgetBlueprint /FortniteCoreUI/Frontend/SimpleBuildAndEditModals/SimpleBuildAndEditIntroModal
graphs [
 BP_GetDesiredFocusTarget {nodes [
   Get Desired Focus Target()
   n1 = Get FortSettingsListViewMVVM() -> (FortSettingsListViewMVVM:Fort Settings List View MVVM Object Reference)
   n2 = GetIndexForItem!(
    self:List View Object Reference=@n1
    Item:Object Reference=Get Selected Item!(self:List View Object Reference=@n1) -> (ReturnValue:Object Reference) {is=title K2Node_CallFunction}
   ) -> (ReturnValue:int=0)
   n3 = Get FortSettingsListViewMVVM()
   > SetSelectedIndex!(
    execute
    self:List View Object Reference=@n3
    Index:int=Select(Option 0:int=0, Option 1:int=@n2, Index="integer >= integer"(A=@n2)) -> (ReturnValue:int)
   ) -> (then)
   > Return Node!(execute, ReturnValue:Widget Object Reference=@n3)
  ]}
 EventGraph {nodes [
   n1 = "On Item Selection Changed (FortSettingsListViewMVVM)"!() -> (OutputDelegate, then, Item:Object Reference, bIsSelected:bool) {is=title K2Node_ComponentBoundEvent}
   > HandleItemSelectionChanged!(
    execute
    self:Simple Build And Edit List View Model Object Reference=Get SimpleBuildAndEditListViewModel() -> (SimpleBuildAndEditListViewModel:Simple Build And Edit List View Model Object Reference)
    Object:Object Reference=@n1.Item
    bIsSelected:bool=@n1.bIsSelected
   ) -> (then)
   n2 = ButtonClicked!() -> (OutputDelegate, then, ButtonIndex:int) {is=title K2Node_CustomEvent}
   > Branch(Condition="Equal (Integer)"(A=@n2.ButtonIndex))
   > HandleBackPressed(self:Simple Build And Edit List View Model Object Reference=Get SimpleBuildAndEditListViewModel())
   > DeactivateWidget()
   n3 = OnButtonClicked!() -> (OutputDelegate, then, Button:Common Button Base Object Reference) {is=title K2Node_CustomEvent}
   > DeactivateWidget(self=Self-Reference())
   > ShowSettingsScreen!(execute, SettingsTabId:Name=GameplayCollection, SettingToSelect:String=SimpleBuild) -> (then)
   UpdateFocus!() -> (OutputDelegate, then, NumEntries:int) {is=title K2Node_CustomEvent}
   > SetFocus(self=GetDesiredFocusTarget!(self:Common Activatable Widget Object Reference=Self-Reference()) -> (ReturnValue:Widget Object Reference))
   Event PreConstruct()
   > SetCompactUIStyle() {is=title K2Node_CallFunction}
   Event OnInitialized()
   > Bind Event to On Settings to Show in Modal Initialized(self:Simple Build And Edit List View Model Object Reference=Get SimpleBuildAndEditListViewModel(), Delegate=Create Event()) {is=title K2Node_AddDelegate}
   > Bind Event to On Entries Generated(self:List View Base Object Reference=Get FortSettingsListViewMVVM(), Delegate=Create Event()) {is=title K2Node_AddDelegate}
   Event On Activated()
   > Bind Event to On Clicked(
    self:Common Button Base Object Reference=Get WBP_UIKit_Button_GoSettings() -> (WBP_UIKit_Button_GoSettings:WBP UIKit Button Quiet Object Reference)
    Delegate=@n3
   ) {is=title K2Node_AddDelegate}
   n4 = Contains Item(
    TargetArray:Array of Strings=Get PlatformsToShowMessage() -> (PlatformsToShowMessage:Array of Strings)
    ItemToFind:"String (by ref)"=GetPlatformName!() -> (ReturnValue:String)
   )
   n5 = NOT Boolean(A=IsMobileGame())
   > SetVisibility(
    self=Get HorizontalBox_SB_QuickBarText() -> (HorizontalBox_SB_QuickBarText:Horizontal Box Object Reference)
    InVisibility=Select(Option 0:ESlateVisibility Enum=Collapsed, Option 1:ESlateVisibility Enum=HitTestInvisible, Index=AND Boolean(A=@n4, B=@n5)) -> (ReturnValue:ESlateVisibility Enum)
   )
   SetDialogData!(
    execute
    self:WBP UIKit Dialog Internal Object Reference=Get WBP_UIKit_Dialog_Base() -> (WBP_UIKit_Dialog_Base:WBP UIKit Dialog Internal Object Reference)
    Data:UI Kit Dialog Viewmodel Object Reference=Get SimpleBuildAndEdit_UIKitDialogViewModel() -> (SimpleBuildAndEdit_UIKitDialogViewModel:Simple Build And Edit UIKit Dialog View Model Object Reference)
   ) -> (then)
  ]}
 OnSettingsToShowInitialized {nodes [
   OnSettingsToShowInitialized!() -> (then, bShowsSimpleBuild:bool=~, bShowsSimpleEdit:bool=~) {is=title K2Node_FunctionEntry}
   > n1 = Sequence!(execute) -> (then_0, then_1, then_2:Exec, then_3:Exec)
   n2 = Select(Option 0=Collapsed, Option 1=HitTestInvisible, Index=Get bShowsSimpleBuild() -> (bShowsSimpleBuild:bool))
   > SetVisibility(self=Get VerticalBox_SimpleBuild() -> (VerticalBox_SimpleBuild:Vertical Box Object Reference), InVisibility=@n2)
   > SetVisibility(self=Get SizeBoxSimpleBuildVideo() -> (SizeBoxSimpleBuildVideo:Size Box Object Reference), InVisibility=@n2)
   SetDialogData(execute=@n1.then_2, self=Get WBP_UIKit_Dialog_Base(), Data=Get SimpleBuildAndEdit_UIKitDialogViewModel())
   > Bind Event to Dialog Button Clicked Dispatcher(self:WBP UIKit Dialog Internal Object Reference=Get WBP_UIKit_Dialog_Base(), Delegate=Create Event()) {is=title K2Node_AddDelegate}
   n3 = Select(Option 0=Collapsed, Option 1=HitTestInvisible, Index=Get bShowsSimpleEdit() -> (bShowsSimpleEdit:bool))
   SetVisibility(execute=@n1.then_1, self=Get HorizontalBox_SimpleEdit() -> (HorizontalBox_SimpleEdit:Horizontal Box Object Reference), InVisibility=@n3)
   > SetVisibility(self=Get SizeBoxSimpleEditVideo() -> (SizeBoxSimpleEditVideo:Size Box Object Reference), InVisibility=@n3)
   n4 = Get SimpleBuildAndEditListViewModel()
   SetVideoWidget!(
    execute=@n1.then_3
    self:Simple Build And Edit List View Model Object Reference=@n4
    VideoWidget:Fort Video Widget Base Object Reference=Select(
     Option 0:Fort Video Widget Base Object Reference=Get B_FortVideoSimpleEditVideo() -> (B_FortVideoSimpleEditVideo:B Fort Video Simple Build And Edit Modal Object Reference)
     Option 1:Fort Video Widget Base Object Reference=Get B_FortVideoSimpleBuildVideo() -> (B_FortVideoSimpleBuildVideo:B Fort Video Simple Build And Edit Modal Object Reference)
     Index=Get bShowsSimpleBuild()
    ) -> (ReturnValue:Fort Video Widget Base Object Reference)
   ) -> (then)
   > PlayVideoForCurrentSetting(self:Simple Build And Edit List View Model Object Reference=@n4)
  ]}
 SetCompactUIStyle {nodes [
   SetCompactUIStyle() {is=title K2Node_FunctionEntry}
   > Branch(Condition=IsUsingCompactUI())
   > SetPadding!(
    execute
    self:Overlay Slot Object Reference=SlotAsOverlaySlot(Widget=Get Panel() -> (Panel:Horizontal Box Object Reference))
    InPadding_Left:float=0
    InPadding_Top:float=50
    InPadding_Right:float=0
    InPadding_Bottom:float=0
   ) -> (then)
   > Set Size!(
    execute
    Size:Vector 2D Structure=(X=0,Y=33)
    self:Spacer Object Reference=Get Spacer_FinePrints() -> (Spacer_FinePrints:Spacer Object Reference)
   ) -> (then, Output_Get:Vector 2D Structure)
   > SetSize!(
    execute
    self:Spacer Object Reference=Get Spacer_BuildModeInfo() -> (Spacer_BuildModeInfo:Spacer Object Reference)
    InSize:Vector 2D Structure=(X=0,Y=20)
   ) -> (then)
  ]}
]
variables [
 IntroSoundCue {type=SoundBase, default=/Game/Sounds/Fort_UI_Sounds/FrontEnd/New_FrontEnd/Ranked/Fort_UI_Ranked_Whoosh_Intro_Cue}
 PlatformsToShowMessage {type="array<string>", default []}
]
widgets {
 info {
  inheritedWidgetCount=0
  namedSlotCount=2
  parentClass=&CommonActivatableWidget
  rootWidgetClass=/FortUILibrary/Systems/Dialog/WBP_UIKit_Dialog_Internal.WBP_UIKit_Dialog_Internal_C
  widgetCount=31
 }
 widgets [
  WBP_UIKit_Dialog_Base {
   widgetClassPath=/FortUILibrary/Systems/Dialog/WBP_UIKit_Dialog_Internal
   bIsVariable=true
   widget=WBP_UIKit_Dialog_Base
   properties {
    bIsFocusable=true
    useText=Common Rich Text
    buttons {
     each Button Horizontal Fill=true
     each Button Min Width=0
     each Button Size=Small
     focuse Button=0
     space Between Buttons=30
     cTA Buttons Types [Regular,Quiet]
     group Placement {
      horizontal Alignment=HAlign_Left
      vertical Alignment=VAlign_Bottom
      grid {column=1, column Span=0, layer=0, row=2, row Span=0, nudge=<0 0>}
      padding=<ltrb 80 64 80 48>
     }
    }
    dialogSIze {
     height Override {enable=true, value=700}
     max Aspect Ratio {enable=false, value=0}
     max Desired Height {enable=true, value=800}
     max Desired Width {enable=false, value=0}
     min Aspect Ratio {enable=false, value=0}
     min Desired Height {enable=false, value=0}
     min Desired Width {enable=false, value=0}
     width Override {enable=true, value=1424}
    }
    imageProperties {
     material=/FortUILibrary/Systems/Dialog/MaterialInstances/MI_UI_DialogImage
     image Size=<512 0>
     placement {
      horizontal Alignment=HAlign_Fill
      vertical Alignment=VAlign_Fill
      grid {column=1, column Span=0, layer=20, row=0, row Span=1, nudge=<0 0>}
      padding=<ltrb 0 0 0 0>
     }
    }
    textProperties {
     auto Wrap Text=true
     justification=Left
     line Height Percentage=1
     mobile Font Size Multiplier=1.3
     transform Policy=None
     wrap Text At=0
     color {colorUseRule=UseColor_Specified, specifiedColor=<rgba 1 1 1 1>}
     font {
      bForceMonospaced=false
      bMaterialIsStencil=false
      fontMaterial=null
      fontObject=&HeadingNow
      letterSpacing=0
      monospacedWidth=1
      size=18
      skewAmount=0
      typefaceFontName=76 Bold Italic
      outlineSettings {
       bApplyOutlineToDropShadows=false
       bMiteredCorners=false
       bSeparateFillAlpha=false
       outlineMaterial=null
       outlineSize=0
       outlineColor=<rgba 0 0 0 1>
      }
     }
     placement {
      horizontal Alignment=HAlign_Fill
      vertical Alignment=VAlign_Top
      grid {column=1, column Span=0, layer=0, row=1, row Span=0, nudge=<0 0>}
      padding=<ltrb 80 0 80 0>
     }
    }
   }
  }
  Overlay_87 {widgetClassPath=&Overlay, namedSlotHost=WBP_UIKit_Dialog_Base, widget=Overlay_87}
  Panel {
   widgetClassPath=&HorizontalBox
   bIsVariable=true
   parent=Overlay_87
   widget=Panel
   slot {class=OverlaySlot, properties {horizontalAlignment=HAlign_Fill, verticalAlignment=VAlign_Fill}}
  }
  Content {
   widgetClassPath=&VerticalBox
   parent=Panel
   widget=Content
   slot {class=HorizontalBoxSlot, properties {verticalAlignment=VAlign_Center, padding=<ltrb 70 20 64 0>, size {sizeRule=Fill, value=1}}}
  }
  Description {widgetClassPath=&VerticalBox, parent=Content, widget=Description, slot {class=VerticalBoxSlot}}
  HorizontalBox_SimpleEdit {
   widgetClassPath=&HorizontalBox
   bIsVariable=true
   parent=Description
   widget=HorizontalBox_SimpleEdit
   slot {class=VerticalBoxSlot}
  }
  Icon_SimpleEdit {
   widgetClassPath=&Image
   parent=HorizontalBox_SimpleEdit
   widget=Icon_SimpleEdit
   properties {
    visibility=HitTestInvisible
    brush {
     bIsDynamicallyLoaded=false
     drawAs=Image
     imageType=FullColor
     mirroring=NoMirror
     resourceName=None
     resourceObject=/FortniteCoreUI/Frontend/SimpleBuildAndEditModals/Textures/T_UI_Icon_Edit
     tiling=NoTile
     imageSize=<64 64>
     margin=<ltrb 0 0 0 0>
     outlineSettings {
      bUseBrushTransparency=false
      roundingType=HalfHeightRadius
      width=0
      color {colorUseRule=UseColor_Specified, specifiedColor=<rgba 0 0 0 0>}
      cornerRadii=<0 0 0 0>
     }
     tintColor {colorUseRule=UseColor_Specified, specifiedColor=<rgba 1 1 1 1>}
     uVRegion {bIsValid=false, max=<0 0>, min=<0 0>}
    }
   }
   slot {class=HorizontalBoxSlot, properties {horizontalAlignment=HAlign_Center, verticalAlignment=VAlign_Center}}
  }
  Spacer_164 {
   widgetClassPath=&Spacer
   parent=HorizontalBox_SimpleEdit
   widget=Spacer_164
   properties {size=<16 1>}
   slot {class=HorizontalBoxSlot}
  }
  SizeBox_2 {
   widgetClassPath=&SizeBox
   parent=HorizontalBox_SimpleEdit
   widget=SizeBox_2
   properties {bOverride_WidthOverride=true, widthOverride=600}
   slot {class=HorizontalBoxSlot, properties {verticalAlignment=VAlign_Center, size {sizeRule=Fill, value=1}}}
  }
  CommonTextBlock {
   widgetClassPath=&CommonTextBlock
   parent=SizeBox_2
   widget=CommonTextBlock
   properties {
    autoWrapText=true
    text=Enable Simple Edit in settings to change building pieces with only one button press.
    font {
     bForceMonospaced=false
     bMaterialIsStencil=false
     fontMaterial=null
     fontObject=&HeadingNow
     letterSpacing=0
     monospacedWidth=1
     size=13.5
     skewAmount=0
     typefaceFontName=74 Regular
     outlineSettings {
      bApplyOutlineToDropShadows=false
      bMiteredCorners=false
      bSeparateFillAlpha=false
      outlineMaterial=null
      outlineSize=0
      outlineColor=<rgba 0 0 0 1>
     }
    }
    shadowOffset=<0 0>
   }
   slot {class=SizeBoxSlot}
  }
  VerticalBox_SimpleBuild {
   widgetClassPath=&VerticalBox
   bIsVariable=true
   parent=Description
   widget=VerticalBox_SimpleBuild
   slot {class=VerticalBoxSlot, properties {padding=<ltrb 0 12 0 0>}}
  }
  HorizontalBox_SB_MainText {
   widgetClassPath=&HorizontalBox
   bIsVariable=true
   parent=VerticalBox_SimpleBuild
   widget=HorizontalBox_SB_MainText
   slot {class=VerticalBoxSlot}
  }
  Icon_SimpleBuild {
   widgetClassPath=&Image
   parent=HorizontalBox_SB_MainText
   widget=Icon_SimpleBuild
   properties {
    visibility=HitTestInvisible
    brush {
     bIsDynamicallyLoaded=false
     drawAs=Image
     imageType=FullColor
     mirroring=NoMirror
     resourceName=None
     resourceObject=/FortniteCoreUI/Frontend/SimpleBuildAndEditModals/Textures/T_UI_Icon_Build2
     tiling=NoTile
     imageSize=<64 64>
     margin=<ltrb 0 0 0 0>
     outlineSettings {
      bUseBrushTransparency=false
      roundingType=HalfHeightRadius
      width=0
      color {colorUseRule=UseColor_Specified, specifiedColor=<rgba 0 0 0 0>}
      cornerRadii=<0 0 0 0>
     }
     tintColor {colorUseRule=UseColor_Specified, specifiedColor=<rgba 1 1 1 1>}
     uVRegion {bIsValid=false, max=<0 0>, min=<0 0>}
    }
   }
   slot {class=HorizontalBoxSlot, properties {horizontalAlignment=HAlign_Center, verticalAlignment=VAlign_Center}}
  }
  Spacer {
   widgetClassPath=&Spacer
   parent=HorizontalBox_SB_MainText
   widget=Spacer
   properties {size=<16 1>}
   slot {class=HorizontalBoxSlot}
  }
  SizeBox_1 {
   widgetClassPath=&SizeBox
   parent=HorizontalBox_SB_MainText
   widget=SizeBox_1
   properties {bOverride_WidthOverride=true, widthOverride=510}
   slot {class=HorizontalBoxSlot, properties {verticalAlignment=VAlign_Center, size {sizeRule=Fill, value=1}}}
  }
  CommonTextBlock_1 {
   widgetClassPath=&CommonTextBlock
   parent=SizeBox_1
   widget=CommonTextBlock_1
   properties {
    autoWrapText=true
    text=Simple Build reduces building to two buttons by using your look direction to decide what piece to build.
    font {
     bForceMonospaced=false
     bMaterialIsStencil=false
     fontMaterial=null
     fontObject=&HeadingNow
     letterSpacing=0
     monospacedWidth=1
     size=13.5
     skewAmount=0
     typefaceFontName=74 Regular
     outlineSettings {
      bApplyOutlineToDropShadows=false
      bMiteredCorners=false
      bSeparateFillAlpha=false
      outlineMaterial=null
      outlineSize=0
      outlineColor=<rgba 0 0 0 1>
     }
    }
    shadowOffset=<0 0>
   }
   slot {class=SizeBoxSlot}
  }
  Spacer_BuildModeInfo {
   widgetClassPath=&Spacer
   bIsVariable=true
   parent=VerticalBox_SimpleBuild
   widget=Spacer_BuildModeInfo
   properties {size=<0 30>}
   slot {class=VerticalBoxSlot}
  }
  HorizontalBox_SB_QuickBarText {
   widgetClassPath=&HorizontalBox
   bIsVariable=true
   parent=VerticalBox_SimpleBuild
   widget=HorizontalBox_SB_QuickBarText
   slot {class=VerticalBoxSlot}
  }
  Spacer_1 {
   widgetClassPath=&Spacer
   parent=HorizontalBox_SB_QuickBarText
   widget=Spacer_1
   properties {size=<16 1>}
   slot {class=HorizontalBoxSlot}
  }
  SizeBox {
   widgetClassPath=&SizeBox
   parent=HorizontalBox_SB_QuickBarText
   widget=SizeBox
   properties {bOverride_WidthOverride=true, widthOverride=600}
   slot {class=HorizontalBoxSlot, properties {verticalAlignment=VAlign_Center, size {sizeRule=Fill, value=1}}}
  }
  CommonTextBlock_2 {
   widgetClassPath=&CommonTextBlock
   parent=SizeBox
   widget=CommonTextBlock_2
   properties {
    autoWrapText=true
    mobileFontSizeMultiplier=1.2
    text="If you are playing on keyboard && mouse, it is recommended that Switch Mode is bound to a valid key in order for Simple Build to work properly. "
    colorAndOpacity {colorUseRule=UseColor_Specified, specifiedColor=<rgba 0.9301 1 0.0103 1>}
    font {
     bForceMonospaced=false
     bMaterialIsStencil=false
     fontMaterial=null
     fontObject=&HeadingNow
     letterSpacing=0
     monospacedWidth=1
     size=10.5
     skewAmount=0
     typefaceFontName=74 Regular
     outlineSettings {
      bApplyOutlineToDropShadows=false
      bMiteredCorners=false
      bSeparateFillAlpha=false
      outlineMaterial=null
      outlineSize=0
      outlineColor=<rgba 0 0 0 1>
     }
    }
    shadowOffset=<0 0>
   }
   slot {class=SizeBoxSlot}
  }
  Spacer_FinePrints {
   widgetClassPath=&Spacer
   bIsVariable=true
   parent=Content
   widget=Spacer_FinePrints
   properties {size=<1 50>}
   slot {class=VerticalBoxSlot}
  }
  VerticalBox_0 {
   widgetClassPath=&VerticalBox
   parent=Content
   widget=VerticalBox_0
   properties {bOverride_Cursor=true}
   slot {class=VerticalBoxSlot}
  }
  FortSettingsListViewMVVM {
   widgetClassPath=/Script/FortniteUI.FortSettingsListViewMVVM
   bIsVariable=true
   parent=VerticalBox_0
   widget=FortSettingsListViewMVVM
   properties {
    bAllowDragging=false
    bReturnFocusToSelection=true
    clipping=Inherit
    entryWidgetClass=/FortniteCoreUI/Settings/SettingEntries/SettingsListEntry_Discrete_WithMVVM.SettingsListEntry_Discrete_WithMVVM_C
    numDesignerPreviewEntries=1
    verticalEntrySpacing=16
    visualData=&/SimpleBuildAndEditModalVisualData
   }
   slot {class=VerticalBoxSlot}
  }
  WBP_UIKit_Button_GoSettings {
   widgetClassPath=/FortUILibrary/UIKit/Buttons/WBP_UIKit_Button_Quiet
   bIsVariable=true
   parent=VerticalBox_0
   widget=WBP_UIKit_Button_GoSettings
   properties {maxHeight=40, maxWidth=220, text=Go Settings}
   slot {class=VerticalBoxSlot, properties {horizontalAlignment=HAlign_Left, verticalAlignment=VAlign_Center, padding=<ltrb 10 10 0 0>}}
  }
  Video {widgetClassPath=&Overlay, namedSlotHost=WBP_UIKit_Dialog_Base, widget=Video}
  SizeBoxSimpleEditVideo {
   widgetClassPath=&SizeBox
   bIsVariable=true
   parent=Video
   widget=SizeBoxSimpleEditVideo
   properties {bOverride_HeightOverride=true, bOverride_WidthOverride=true, heightOverride=704, widthOverride=512}
   slot {class=OverlaySlot}
  }
  B_FortVideoSimpleEditVideo {
   widgetClassPath=&/B_FortVideoSimpleBuildAndEditModal
   bIsVariable=true
   parent=SizeBoxSimpleEditVideo
   widget=B_FortVideoSimpleEditVideo
   properties {activeSoundMix=null, bIsFocusable=false, bSupportsActivationFocus=false, visibility=HitTestInvisible}
   slot {class=SizeBoxSlot, properties {horizontalAlignment=HAlign_Left, verticalAlignment=VAlign_Top}}
  }
  SizeBoxSimpleBuildVideo {
   widgetClassPath=&SizeBox
   bIsVariable=true
   parent=Video
   widget=SizeBoxSimpleBuildVideo
   properties {bOverride_HeightOverride=true, bOverride_WidthOverride=true, heightOverride=704, widthOverride=512}
   slot {class=OverlaySlot}
  }
  B_FortVideoSimpleBuildVideo {
   widgetClassPath=&B_FortVideoSimpleBuildAndEditModal
   bIsVariable=true
   parent=SizeBoxSimpleBuildVideo
   widget=B_FortVideoSimpleBuildVideo
   properties {
    activeSoundMix=null
    bIsFocusable=false
    bSupportsActivationFocus=false
    mediaTexture=/FortniteCoreUI/Frontend/SimpleBuildAndEditModals/VideoFiles/WidgetMediaPlayer_Video_SB
    shouldOverrideImageVideoTextureBrush=true
    visibility=HitTestInvisible
   }
   slot {class=SizeBoxSlot, properties {horizontalAlignment=HAlign_Left, verticalAlignment=VAlign_Top}}
  }
  FallbackImage {
   widgetClassPath=&FortLazyImage
   bIsVariable=true
   parent=Video
   widget=FallbackImage
   properties {brush {
     bIsDynamicallyLoaded=false
     drawAs=Image
     imageType=FullColor
     mirroring=NoMirror
     resourceName=None
     resourceObject=null
     tiling=NoTile
     imageSize=<512 704>
     margin=<ltrb 0 0 0 0>
     outlineSettings {
      bUseBrushTransparency=false
      roundingType=HalfHeightRadius
      width=0
      color {colorUseRule=UseColor_Specified, specifiedColor=<rgba 0 0 0 0>}
      cornerRadii=<0 0 0 0>
     }
     tintColor {colorUseRule=UseColor_Specified, specifiedColor=<rgba 1 1 1 1>}
     uVRegion {bIsValid=false, max=<0 0>, min=<0 0>}
    }}
   slot {class=OverlaySlot, properties {horizontalAlignment=HAlign_Center, verticalAlignment=VAlign_Fill}}
  }
 ]
}
